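# Physics asset for the Bunny Fashion Female two-handed pickaxe cosmetic, giving its bunny ears and feet limited physical motion relative to the handle.
# Capsule bodies cover the handle, a four-bone chain in each ear, and a single feet bone.
# Each ear chain and the feet constrain back to the handle, with ear swing limits widening toward the tips while twist stays tight throughout.
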PhysicsAsset /BRCosmetics/Weapons/FORT_Melee/Pickaxe_Bunny_Fashion_Female/Meshes/Bunny_Fashion_Female_Physics
bodies {
 Melee_TwoHanded_Handle [{shapeType=Capsule, length=73.6732, radius=25.0478, center=<0.0544 -4.7795 33.0209>, rotation=<rot 0.428 -90 0.2297>}]
 ear_1_l [{shapeType=Capsule, length=4.1514, radius=3.9253, center=<8.3541 0.7346 1.0931>, rotation=<rot 31.4022 -91.2792 -90>}]
 ear_1_r [{shapeType=Capsule, length=4.3024, radius=4.0644, center=<8.2733 -0.576 1.3458>, rotation=<rot 21.0548 91.2006 -90>}]
 ear_2_l [{shapeType=Capsule, length=5.2574, radius=4.1977, center=<6.6421 0.1138 0.3257>, rotation=<rot 0 -91.5107 -86.8444>}]
 ear_2_r [{shapeType=Capsule, length=5.3866, radius=4.4177, center=<7.2245 -0.2237 0.267>, rotation=<rot 0 -92.6685 -84.6358>}]
 ear_3_l [{shapeType=Capsule, length=3.7387, radius=3.3888, center=<7.2783 -0.0426 -0.319>, rotation=<rot 0 -92.501 -83.4862>}]
 ear_3_r [{shapeType=Capsule, length=3.7373, radius=3.5331, center=<7.5749 0.0359 -0.4307>, rotation=<rot 0 -87.5139 -82.5502>}]
 ear_4_l [{shapeType=Capsule, length=0.505, radius=0.505, rotation=<rot 0 -90 0>}]
 ear_4_r [{shapeType=Capsule, length=0.505, radius=0.505, rotation=<rot 0 -90 0>}]
 feet [{shapeType=Capsule, length=3.4379, radius=16.275, center=<-0.0001 19.7431 0>, rotation=<rot 0 0 90>}]
}
constraints [
 {bone1Name=ear_1_r, bone2Name=Melee_TwoHanded_Handle, swing1LimitDegrees=20, swing2LimitDegrees=20, twistLimitDegrees=5}
 {bone1Name=ear_2_r, bone2Name=ear_1_r, swing1LimitDegrees=45, swing2LimitDegrees=45, twistLimitDegrees=5}
 {bone1Name=ear_3_r, bone2Name=ear_2_r, swing1LimitDegrees=60, swing2LimitDegrees=60, twistLimitDegrees=5}
 {bone1Name=ear_4_r, bone2Name=ear_3_r, swing1LimitDegrees=60, swing2LimitDegrees=60, twistLimitDegrees=5}
 {bone1Name=ear_1_l, bone2Name=Melee_TwoHanded_Handle, swing1LimitDegrees=20, swing2LimitDegrees=20, twistLimitDegrees=5}
 {bone1Name=ear_2_l, bone2Name=ear_1_l, swing1LimitDegrees=45, swing2LimitDegrees=45, twistLimitDegrees=5}
 {bone1Name=ear_3_l, bone2Name=ear_2_l, swing1LimitDegrees=60, swing2LimitDegrees=60, twistLimitDegrees=5}
 {bone1Name=ear_4_l, bone2Name=ear_3_l, swing1LimitDegrees=60, swing2LimitDegrees=60, twistLimitDegrees=5}
 {bone1Name=feet, bone2Name=Melee_TwoHanded_Handle, swing1LimitDegrees=45, swing2LimitDegrees=45, twistLimitDegrees=5}
]
