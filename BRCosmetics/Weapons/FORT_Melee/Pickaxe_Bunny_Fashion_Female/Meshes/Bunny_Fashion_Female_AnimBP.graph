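# Animation Blueprint for the female Bunny Fashion pickaxe, adding secondary motion to the weapon mesh.
# The cached input pose feeds two branches, selected by bRigidBodyEnabled: RigidBody simulation relative to the two-handed handle bone, or a spring-and-trail fallback.
# The fallback chains a faint spring on the feet bone into trail controllers for both four-bone ear chains, with rotation limits widening along each chain.
# The event graph holds no logic.
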AnimBlueprint /BRCosmetics/Weapons/FORT_Melee/Pickaxe_Bunny_Fashion_Female/Meshes/Bunny_Fashion_Female_AnimBP
target_skeleton=&/Bunny_Fashion_Female_Skeleton
graphs [
 AnimGraph {nodes [
   n1 = RigidBody(ComponentPose=Local To Component(LocalPose=Use cached pose 'pose_IN'())) {
    bClampLinearTranslationLimitToRefPose=true
    bOverrideWorldGravity=true
    overridePhysicsAsset=&/Bunny_Fashion_Female_Physics
    simulationSpace=BaseBoneSpace
    alphaBoolBlend {blendOption=Linear}
    baseBoneRef {boneName=Melee_TwoHanded_Handle}
    componentAppliedLinearAccClamp=<1000 1000 1000>
    componentLinearAccScale=<0.5 0.5 0.45>
    componentLinearVelScale=<0.5 0.5 0.3>
    overrideWorldGravity {z=-980}
   }
   n2 = Spring controller - Bone: feet!(ComponentPose:Component Space Pose Link Structure=Local To Component(LocalPose=Use cached pose 'pose_IN'()), Alpha:float=0.03) -> (Pose:Component Space Pose Link Structure) {
    is=title AnimGraphNode_SpringBone
    bLimitDisplacement=true
    bRotateX=true
    bRotateY=true
    bRotateZ=true
    maxDisplacement=200
    alphaBoolBlend {blendOption=Linear}
    springBone {boneName=feet}
   }
   n3 = Trail controller - Bone: ear_4_l(ComponentPose=@n2, Alpha=1) {
    bActorSpaceFakeVel=true
    bEnableDebug=true
    bLimitRotation=true
    bShowTrailLocation=true
    chainLength=4
    maxDeltaTime=0.0333
    alphaScaleBiasClamp {scale=0.25}
    baseJoint {boneName=Melee_TwoHanded_Handle}
    rotationLimits [
     {limitMax=<5 5 5>, limitMin=<-5 -5 -5>}
     {limitMax=<5 10 10>, limitMin=<-5 -10 -10>}
     {limitMax=<5 20 20>, limitMin=<-5 -20 -20>}
     {limitMax=<5 30 30>, limitMin=<-5 -30 -30>}
    ]
    rotationOffsets [<0 0 0>, <0 0 0>, <0 0 0>, <0 0 0>]
    trailBone {boneName=ear_4_l}
   }
   n4 = Trail controller - Bone: ear_4_r(ComponentPose=@n3, Alpha=1) {
    bActorSpaceFakeVel=true
    bEnableDebug=true
    bLimitRotation=true
    bShowTrailLocation=true
    chainLength=4
    maxDeltaTime=0.0333
    alphaScaleBiasClamp {scale=0.25}
    baseJoint {boneName=Melee_TwoHanded_Handle}
    rotationLimits [
     {limitMax=<5 5 5>, limitMin=<-5 -5 -5>}
     {limitMax=<5 10 10>, limitMin=<-5 -10 -10>}
     {limitMax=<5 20 20>, limitMin=<-5 -20 -20>}
     {limitMax=<5 30 30>, limitMin=<-5 -30 -30>}
    ]
    rotationOffsets [<0 0 0>, <0 0 0>, <0 0 0>, <0 0 0>]
    trailBone {boneName=ear_4_r}
   }
   Output Pose(Result=Blend Poses by bool(BlendPose_0=Component To Local(ComponentPose=@n1), BlendPose_1=Component To Local(ComponentPose=@n4), bActiveValue=Get bRigidBodyEnabled()) {blendType=Linear}) {layerGroup=None}
   AnimGraphNode_SubInput_0 = Input Pose()
   Save cached pose 'pose_IN'(Pose=@AnimGraphNode_SubInput_0) {cacheName=pose_IN}
  ]}
 EventGraph {nodes [
   TryGetPawnOwner()
   Event BlueprintUpdateAnimation()
  ]}
]
variables []
